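# Masked, dithered material for Niagara mesh particles in the Polygon player body's FX, colored by particle color.
# The SuperLevel dynamic parameter drives multi-step lerps that hue-shift, desaturate and brighten that color.
# The HDR dynamic parameter scales it through DeriveHDRfromLDR, feeding MF_FX_EmissiveAdjustments for emissive.
# A clamped sum of second-UV green and a particle-alpha-driven offset forms a wipe-style mask value.
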Material /Game/Characters/Player/Male/Medium/Bodies/M_MED_Polygon/FX/M_Polygon3D
graph {
 properties {
  bUsedWithNiagaraMeshParticles=true
  blendMode=BLEND_Masked
  ditherOpacityMask=true
  pixelDepthOffsetMode=PDOM_Legacy
  translucencyLightingMode=TLM_VolumetricPerVertexDirectional
  usedShadingModels=Default Lit
 }
 nodes [
  n1 = TextureCoordinate() {coordinateIndex=1}
  n2 = ParticleColor()
  Clamp(None=Add(A=ComponentMask(None=@n1) {g=true}, B=LinearInterpolate(Alpha=@n2.A) {constA=-1}))
  Constant()
  Constant()
  Constant() {r=1}
  n3 = Constant()
  n4 = Constant() {r=0.05}
  n5 = Constant() {r=0.1}
  n6 = Divide(A=DynamicParameter() {paramNames [SuperLevel,HDR,Param3,Param4]}.SuperLevel) {constB=3}
  n7 = Constant()
  n8 = Constant() {r=0.5}
  n9 = MaterialFunctionCall!(1=@n7, 2=@n7, 3=@n8, 4=@n8, A=@n6) {materialFunction=/Engine/Functions/Engine_MaterialFunctions02/Lerp_Multiple_Float}
  n10 = Desaturation(
   None=MaterialFunctionCall(
    Hue Shift Percentage=MaterialFunctionCall(1=@n3, 2=@n4, 3=@n5, 4=@n5, A=@n6) {materialFunction=&Lerp_Multiple_Float}.Lerp 4 Inputs
    Texture=ParticleColor().RGB
   ) {materialFunction=&HueShift}.Result
   Fraction=@n9.Lerp 4 Inputs
  )
  n11 = Multiply(A=@n9.Lerp 4 Inputs) {constB=2}
  n12 = Multiply(A=Add(A=@n10, B=Multiply(A=@n10, B=@n11))) {constB=0.3}
  MaterialFunctionCall(Emissive=MaterialFunctionCall(LDR Input=Saturate(None=@n12), HDR Tint=@n12, HDR Intensity Multiplier=DynamicParameter() {paramNames [SuperLevel,HDR,Param3,Param4]}.HDR) {materialFunction=&DeriveHDRfromLDR}.RemappedRange) {materialFunction=&MF_FX_EmissiveAdjustments}
 ]
}
