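# Simple translucent, unlit, two-sided FX material for the Polygon player character, usable on Niagara mesh particles.
# Particle color drives emissive: clamped RGB feeds DeriveHDRfromLDR, tinted by that color and scaled by the HDR Intensity dynamic parameter, then passes through MF_FX_EmissiveAdjustments.
# Particle alpha is multiplied by a near-camera fade with an adjustable distance.
# The Erode dynamic parameter is named but unused.
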Material /Game/Characters/Player/Male/Medium/Bodies/M_MED_Polygon/FX/M_PolygonSimple
graph {
 properties {
  bUsedWithNiagaraMeshParticles=true
  blendMode=BLEND_Translucent
  pixelDepthOffsetMode=PDOM_Legacy
  shadingModel=MSM_Unlit
  translucencyLightingMode=TLM_VolumetricPerVertexDirectional
  twoSided=true
  usedShadingModels=Unlit
 }
 nodes [
  n1 = ParticleColor()
  MaterialFunctionCall(Emissive=MaterialFunctionCall(
    LDR Input=Clamp(None=@n1.RGB)
    HDR Tint=@n1.RGB
    HDR Intensity Multiplier=DynamicParameter() {defaultValue=<rgba 1 0.55 1 1>, paramNames [HDR Intensity, Erode, Param3, Param4]}.HDR Intensity
   ) {materialFunction=&DeriveHDRfromLDR}.RemappedRange) {materialFunction=&MF_FX_EmissiveAdjustments}
  Multiply(
   A=@n1.A
   B=MaterialFunctionCall(Distance=ScalarParameter() {defaultValue=350, parameterName=Near Camera Distance}) {materialFunction=&Near_Camera_Fade}.Result
  )
 ]
}
