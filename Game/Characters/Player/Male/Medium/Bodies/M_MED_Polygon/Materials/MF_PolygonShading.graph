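# Material function shading the Polygon player character, with super-level color stages and a transition effect.
# It modifies input attributes with hue shift, desaturation, secondary colors and pulsing Fresnel, splitting color between base and emissive.
# Reroutes supply sun-dot-normal fake lighting, a time-of-day curve, channel masks, and a height sweep with glow band broken up by scrolling world-aligned texture.
# The Hair Wind switch adds masked sine/cosine vertex sway from global wind.
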MaterialFunction /Game/Characters/Player/Male/Medium/Bodies/M_MED_Polygon/Materials/MF_PolygonShading
graph {nodes [
  Fresnel(ExponentIn=ScalarParameter() {defaultValue=0.1, parameterName=Desaturate Fresnel Exponent}) {exponent=0.1}
  n1 = Saturate(None=LinearInterpolate(
    A=ScalarParameter() {defaultValue=0.4, parameterName=Emissive Diffuse Mix Night}
    B=ScalarParameter() {defaultValue=0.6, parameterName=Emissive Diffuse Mix}
    Alpha=NamedRerouteUsage()
   ) {constA=0.5})
  n2 = ScalarParameter() {defaultValue=0.25, parameterName=ShadingDarkness}
  n3 = NamedRerouteUsage()
  n4 = Constant()
  n5 = VectorParameter() {desc=Neck color, parameterName=Stage 1 Sec Color, defaultValue=<rgba 0.5 0.0182 0.01 1>}
  n6 = VectorParameter() {desc=Neck color, parameterName=Stage 2 Sec Color, defaultValue=<rgba 1 0.267 0.0595 1>}
  n7 = NamedRerouteUsage()
  n8 = NamedRerouteUsage()
  n9 = NamedRerouteUsage()
  n10 = NamedRerouteUsage()
  n11 = GetMaterialAttributes(None=FunctionInput() {inputName=Input, inputType=FunctionInput_MaterialAttributes}) {attributeGetTypes [#1, #2]}
  n12 = Constant()
  n13 = ScalarParameter() {defaultValue=0.35, parameterName=Desaturate Amount}
  n14 = MaterialFunctionCall!(1=@n12, 2=@n12, 3=@n13, 4=@n13, A=NamedRerouteUsage()) {materialFunction=/Engine/Functions/Engine_MaterialFunctions02/Lerp_Multiple_Float}
  n15 = Desaturation(
   None=MaterialFunctionCall(Hue Shift Percentage=Multiply(A=@n9, B=@n10), Texture=@n11.BaseColor) {materialFunction=&HueShift}.Result
   Fraction=Multiply(A=NamedRerouteUsage(), B=Multiply(A=@n14.Lerp 4 Inputs))
  )
  n16 = Multiply(A=Multiply(A=Multiply(A=@n14.Lerp 4 Inputs) {constB=2}, B=Fresnel() {exponent=0.1}), B=NamedRerouteUsage())
  n17 = LinearInterpolate(
   A=Add(A=@n15, B=Multiply(A=@n15, B=@n16))
   B=Multiply(
    A=Multiply(A=@n3) {constB=5}
    B=MaterialFunctionCall(1=@n4, 2=@n5.RGB, 3=@n6.RGB, 4=@n6.RGB, A=@n7) {materialFunction=&Lerp_Multiple_Float3}.Lerp 4 Inputs
   )
   Alpha=MaterialFunctionCall(1=Constant(), 2=@n8, 3=@n8, 4=@n8, A=@n7) {materialFunction=&Lerp_Multiple_Float}.Lerp 4 Inputs
  )
  n18 = Multiply(A=LinearInterpolate(A=Multiply(A=@n17, B=@n2), B=@n17, Alpha=NamedRerouteUsage()), B=@n1) {constB=0.7}
  n19 = Constant3Vector() {constant=<rgba 0.718 0.1508 0.0427 1>}
  n20 = Constant3Vector() {constant=<rgba 1 0.513 0.0595 1>}
  n21 = Saturate(None=NamedRerouteUsage())
  n22 = NamedRerouteUsage()
  n23 = Add(A=NamedRerouteUsage(), B=Add(A=Time(), B=Multiply(A=@n22) {constB=-4}))
  n24 = Constant() {r=1}
  n25 = Fresnel(ExponentIn=LinearInterpolate(Alpha=Fmod(A=@n23, B=@n24)) {constA=1, constB=1.5}) {exponent=1}
  n26 = Multiply(
   A=Multiply(A=LinearInterpolate(A=@n19, B=@n20, Alpha=@n21) {constA=0.3}, B=@n25)
   B=LinearInterpolate(Alpha=NamedRerouteUsage()) {constA=0.5}
  )
  n27 = Constant()
  n28 = MaterialFunctionCall(1=@n27, 2=@n27, 3=NamedRerouteUsage(), 4=@n25, A=NamedRerouteUsage()) {materialFunction=&Lerp_Multiple_Float}
  n29 = SetMaterialAttributes!(
   MaterialAttributes=@n11.MaterialAttributes
   EmissiveColor=Add(A=LinearInterpolate(A=@n18, B=@n26, Alpha=@n28.Lerp 4 Inputs), B=NamedRerouteUsage())
   BaseColor=Multiply(A=@n17, B=OneMinus(None=@n1)) {constB=0.3}
   Metallic=Constant()
   Specular=Constant() {r=0.25}
   Roughness=Constant() {r=0.09}
  ) {attributeSetTypes [
    #3
    #1
    #4
    #2
    #5
   ]}
  n30 = TextureSample() {texture=/Game/Characters/Player/Male/Medium/Bodies/M_MED_Polygon/Textures/M_MED_Polygon_Hair_Mask}
  n31 = Add(A=Divide(A=Time()) {constB=1.5}, B=@n30.G)
  n32 = Sine(None=@n31)
  n33 = CollectionParameter() {
   collection=&FortniteMaterialParameters
   parameterName=WindDirection_RGB Speed_A
  }
  n34 = ComponentMask(None=@n33) {b=true, g=true, r=true}
  n35 = Power(Base=@n30.R) {constExponent=0.5}
  n36 = Multiply(A=@n34, B=@n35) {constA=1, constB=10}
  n37 = Cosine(None=@n31)
  n38 = Multiply(A=CrossProduct(A=@n34, B=Constant3Vector() {constant=<rgba 0 0 1 0>}), B=@n35)
  n39 = Constant3Vector() {constant=<rgba 0 0 5 0>}
  n40 = Add(A=Add(A=Multiply(A=@n32, B=@n36), B=Multiply(A=@n37, B=@n38)), B=Multiply(A=Multiply(A=@n35, B=@n39), B=NamedRerouteUsage()))
  FunctionOutput(None=StaticSwitchParameter(
    True=MaterialFunctionCall!(Material=@n29, WorldPositionOffset=@n40) {materialFunction=/Engine/Functions/MaterialLayerFunctions/MatLayerBlend_AddWorldPositionOffset}.Blended Material
    False=@n29
   ) {parameterName=Hair Wind})
  n41 = TextureSampleParameter2D() {parameterName=SL_Mask, texture=/Game/Characters/Player/Male/Medium/Bodies/M_MED_Polygon/Textures/T_M_MED_Polygon_Body_SL}
  NamedRerouteDeclaration(None=@n41.B) {name=SL_MaskBlueChannel, nodeColor=<rgba 0 0.3426 1 1>}
  n42 = CollectionParameter() {
   collection=&FortniteMaterialParameters
   parameterName=SunlightDirectionalVector
  }
  n43 = ComponentMask(None=@n42) {b=true, g=true, r=true}
  n44 = PixelNormalWS()
  NamedRerouteDeclaration(None=Ceil(None=DotProduct(A=@n43, B=@n44))) {name=Fake Lighting, nodeColor=<rgba 1 0.4916 0 1>}
  n45 = CurveAtlasRowParameter(CurveTime=CollectionParameter() {collection=&FortniteMaterialParameters, parameterName=TimeOfDay_0-1}) {
   atlas=/Game/Characters/Shared/CurveAtlas/AthenaCharacterCurveAtlas
   curve=/Game/Characters/Player/Male/Medium/Bodies/M_MED_Polygon/Textures/PolygonTOD_Curve
   defaultValue=20
   group=TOD_Adjustments
   parameterName=TODBrightnessAdjustmentCurve
  }
  NamedRerouteDeclaration(None=@n45.R) {name=TOD Lerp, nodeColor=<rgba 1 0.2344 0 1>}
  n46 = PreSkinnedPosition()
  n47 = ShaderStageSwitch(PixelShader=VertexInterpolator(VS=@n46).PS, VertexShader=@n46)
  n48 = Divide(A=Add(A=ComponentMask(None=@n47) {b=true}) {constB=20}) {constB=235}
  NamedRerouteDeclaration(None=@n48) {name=Height Gradient, nodeColor=<rgba 1 0.9844 0 1>}
  n49 = ScalarParameter() {parameterName=SuperLevel}
  n50 = MaterialFunctionCall(Z=Multiply(A=Time()) {constB=-4}) {materialFunction=&MakeFloat3}
  n51 = PreSkinnedNormal()
  n52 = VertexInterpolator(VS=@n51)
  n53 = ComponentMask(None=MaterialFunctionCall(
    TextureObject=TextureObject() {samplerType=SAMPLERTYPE_LinearColor, texture=/Game/Characters/Player/Male/Medium/Bodies/M_MED_Polygon/Textures/M_MED_Polygon_SLMask}
    TextureSize=Constant3Vector() {constant=<rgba 100 100 100 0>}
    WorldPosition=Add(A=@n47, B=@n50.Result)
    World Space Normal=ShaderStageSwitch(PixelShader=@n52.PS, VertexShader=@n51)
   ) {materialFunction=&WorldAlignedTexture}.XYZ Texture) {r=true}
  n54 = Add(A=@n48, B=ShaderStageSwitch(PixelShader=Multiply(A=@n53) {constB=0.2}, VertexShader=Constant() {r=0.1}))
  n55 = Saturate(None=Subtract(A=Time(), B=ScalarParameter() {parameterName=SuperLevelAnim}))
  n56 = SmoothStep(Min=@n54, Max=@n54, Value=@n55)
  n57 = Subtract(A=@n49)
  NamedRerouteDeclaration(None=Divide(A=LinearInterpolate(A=@n57, B=@n49, Alpha=@n56)) {constB=3}) {name=SL Anim Mask, nodeColor=<rgba 1 0 0.2109 1>}
  n58 = ScalarParameter() {defaultValue=0.1, parameterName=HueChange2}
  NamedRerouteDeclaration(None=MaterialFunctionCall(1=Constant(), 2=ScalarParameter() {defaultValue=0.05, parameterName=HueChange1}, 3=@n58, 4=@n58, A=NamedRerouteUsage()) {materialFunction=&Lerp_Multiple_Float}.Lerp 4 Inputs) {name=Hue Shift, nodeColor=<rgba 1 0.116 0.4317 1>}
  NamedRerouteDeclaration(None=@n41.R) {name=SL_MaskRedChannel, nodeColor=<rgba 1 0 0.0618 1>}
  NamedRerouteDeclaration(None=@n41.G) {name=SL_MaskGreenChannel, nodeColor=<rgba 0.0613 1 0 1>}
  n59 = ScalarParameter() {defaultValue=0.1, parameterName=GlowBlur}
  n60 = ShaderStageSwitch(PixelShader=@n59, VertexShader=ScalarParameter() {defaultValue=0.3, parameterName=BandBlur})
  n61 = Sine(None=SmoothStep(Min=Subtract(A=@n54, B=@n59) {constB=0.1}, Max=Add(A=@n54, B=@n60) {constB=0.1}, Value=@n55))
  NamedRerouteDeclaration(None=MaterialFunctionCall(1=Constant3Vector(), 2=Constant3Vector() {constant=<rgba 0.5938 0 0 0>}, 3=Constant3Vector() {constant=<rgba 1 0.2126 0 0>}, A=@n61) {materialFunction=&Lerp_Multiple_Float3}.Lerp 3 Inputs) {name=SL_Glow, nodeColor=<rgba 1 0 0.2578 1>}
  NamedRerouteDeclaration(None=@n53) {name=GeoPattern, nodeColor=<rgba 0 0.4609 1 1>}
  NamedRerouteDeclaration(None=Saturate(None=@n61)) {name=SL_Band, nodeColor=<rgba 0.2891 1 0 1>}
  NamedRerouteDeclaration(None=@n11.Specular) {name=Spec, nodeColor=<rgba 0.8047 0 1 1>}
  ScalarParameter() {defaultValue=1, parameterName=PrevTransition, sliderMax=1}
 ]}
